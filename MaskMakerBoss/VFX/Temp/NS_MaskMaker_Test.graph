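# Test Niagara system in the MaskMaker boss Temp folder, pairing a glowing mesh with a telegraph decal.
# EdgeGlow bursts mesh particles offset above the system location, then applies camera offset, color dimming and dynamic material parameters.
# TelegraphDecal bursts one vertically oriented decal and computes an emitter opacity value that falls as System.BeamLengthScalar rises through a curve.
# Both emitters keep particles past their lifetime, letting infinitely lived ones die when the emitter deactivates.
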NiagaraSystem /MaskMakerBoss/VFX/Temp/NS_MaskMaker_Test
dependencies {
 usedDataInterfaces [&NiagaraDataInterfaceCurve, &NiagaraDataInterfaceSkeletalMesh]
 usedDynamicInputs [&FN_LOD_Distance_Remap]
 usedModules [
  &self:SystemSpawnScript.SystemScriptSource.SystemScriptGraph.NiagaraNodeAssignment_0.SetVariables_#1
  &SystemState
  &self:SystemSpawnScript.SystemScriptSource.SystemScriptGraph.NiagaraNodeAssignment_1.SetVariables_#2
 ]
}
emitters {
 EdgeGlow {
  emitterUpdateScript [
   &EmitterState(
    Inactive Response=::NewEnumerator1
    Life Cycle Mode=::NewEnumerator1
    Loop Behavior=::NewEnumerator1
    Loop Duration Mode=::NewEnumerator1
    UseLoopDelay=false
   )
   &SpawnBurst_Instantaneous(Use Spawn Probability=false, UseLoopCountLimit=false) {module=SpawnBurst_Instantaneous001}
  ]
  particleSpawnScript [
   &InitializeParticle(Color Mode=::NewEnumerator0, Lifetime=1.5, Mesh Scale Mode=::NewEnumerator3, UsePositionOffset=false, Write Material Random=true)
   &InitialMeshOrientation(Mesh Axis To Orient=<0 0 1>, Mesh Orientation Mode=::NewEnumerator2, Use Rotation Vector=true)
   &SystemLocation(MaskSpawnGroup=false, Offset=<0 0 80>)
  ]
  particleUpdateScript [
   &ParticleState(Kill Particles When Lifetime Has Elapsed=false, Let Infinitely Lived Particles Die When Emitter Deactivates=true)
   &CameraOffset(Camera Offset Amount=38.6542)
   &ScaleColor(Scale RGB=&VectorFromFloat(Value=0.2), ScaleA=true, ScaleRGB=true)
   &DynamicMaterialParameters(Module.Index0Param0Write=true, Module.Index0Param1Write=true, Module.Index0Param2Write=true, Module.Index0Param3Write=true)
  ]
  rendererClasses [&NiagaraMeshRendererProperties]
  renderers [{rendererClass=&NiagaraMeshRendererProperties, rendererIndex=0}]
 }
 TelegraphDecal {
  emitterUpdateScript [
   &EmitterState(Life Cycle Mode=::NewEnumerator1, Loop Behavior=::NewEnumerator1, Loop Duration Mode=::NewEnumerator1, UseLoopDelay=false)
   /SetVariables(Emitter.DecalInitialOpacity:NiagaraFloat=&Lerp_Float(
     A=0.4
     B=0
     Alpha=&FloatFromCurve(FloatCurve=DataInterface() {propertyValues {Curve {keys [<at 0 0 RCIM_Cubic>,<at 1 1 RCIM_Cubic>]}}}, CurveIndex=$System.BeamLengthScalar)
    )) {isSetParameters=true}
   &SpawnBurst_Instantaneous(Spawn Count=1, Use Spawn Probability=false)
  ]
  particleSpawnScript [
   &InitializeParticle(Lifetime=3, Sprite Size Mode=::NewEnumerator3, Uniform Sprite Size=600, UsePositionOffset=false, Write Material Random=true)
   /SetVariables(
    Particles.DecalFade=1
    Particles.DecalOrientation:Quat4f=&OrientMeshToVector(Look At Direction=<0 0 1>, Reference Axis Vector=<1 0 0>)
    Particles.DecalSize:Vector3f=<150 400 400>
   ) {isSetParameters=true}
  ]
  particleUpdateScript [&ParticleState(Kill Particles When Lifetime Has Elapsed=false, Let Infinitely Lived Particles Die When Emitter Deactivates=true)]
  rendererClasses [&NiagaraDecalRendererProperties]
  renderers [{rendererClass=&NiagaraDecalRendererProperties, rendererIndex=0}]
 }
}
summary {
 systemName=NS_MaskMaker_Test
 emitters [
  {emitterName=EdgeGlow, rendererClasses [&NiagaraMeshRendererProperties]}
  {emitterName=TelegraphDecal, rendererClasses [&NiagaraDecalRendererProperties]}
 ]
 userVariables [
  User.DecalLocation {type=NiagaraPosition, description="", defaultValue {}}
  User.ElementSizeScale {type=NiagaraFloat, defaultValue=1, description=""}
  User.IsLocalPlayer {type=NiagaraBool, defaultValue=true}
  User.RandomArray {
   type=NiagaraDataInterfaceArrayFloat
   defaultValue {
    struct=NiagaraExt_VariableValue_DataInterface
    value {
     dataInterface=&self:NiagaraDataInterfaceArrayFloat_0
     dataInterfaceClass=&NiagaraDataInterfaceArrayFloat
    }
   }
  }
  User.Stop Location {type=NiagaraPosition, description="", defaultValue {}}
 ]
}
